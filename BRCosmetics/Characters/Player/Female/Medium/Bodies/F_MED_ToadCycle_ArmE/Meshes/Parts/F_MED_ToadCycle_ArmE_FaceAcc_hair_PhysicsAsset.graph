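# Physics asset for the hair on the ToadCycle_ArmE medium female character's face accessory part, giving its dynamic hair bones simulated bodies.
# Capsule bodies form back, left and right chains rooted at the head, branching into smaller strands, with the left side splitting into front, middle and back.
# Hair joints have narrow twist limits and modest swing limits, looser on some tips.
# Larger head and spine_05 capsules provide body volumes, joined by an angularly free constraint.
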PhysicsAsset /BRCosmetics/Characters/Player/Female/Medium/Bodies/F_MED_ToadCycle_ArmE/Meshes/Parts/F_MED_ToadCycle_ArmE_FaceAcc_hair_PhysicsAsset
bodies {
 dyn_back_hair_01 [{
   shapeType=Capsule
   length=4.75
   radius=1.75
   shapeName=dyn_back_hair_01_capsule
   center=<2.4848 0 -0.1221>
   rotation=<rot -87.1875 -179.9999 -179.9998>
  }]
 dyn_back_hair_02 [{
   shapeType=Capsule
   length=4.75
   radius=1.75
   shapeName=dyn_back_hair_02_capsule
   center=<2.4848 0 -0.1221>
   rotation=<rot -87.1875 -179.9999 -179.9998>
  }]
 dyn_back_hair_s1 [{shapeType=Capsule, length=2, radius=1, shapeName=dyn_back_hair_s1_capsule, center=<2.25 0 0>, rotation=<rot 90 0 0>}]
 dyn_back_hair_s2 [{shapeType=Capsule, length=2, radius=1, shapeName=dyn_back_hair_s2_capsule, center=<2.25 0 0>, rotation=<rot 90 0 0>}]
 dyn_back_hair_s3 [{shapeType=Capsule, length=2, radius=1, shapeName=dyn_back_hair_s3_capsule, center=<2.25 0 0>, rotation=<rot 90 0 0>}]
 dyn_left_hair_01 [{
   shapeType=Capsule
   length=4.75
   radius=1.75
   shapeName=dyn_left_hair_01_capsule
   center=<2.4848 0 -0.1221>
   rotation=<rot -87.1875 179.9991 -179.9989>
  }]
 dyn_left_hair_02 [{
   shapeType=Capsule
   length=4.75
   radius=1.75
   shapeName=dyn_left_hair_02_capsule
   center=<2.4848 0 -0.1221>
   rotation=<rot -78.75 0.0003 0.0001>
  }]
 dyn_left_hair_b [{
   shapeType=Capsule
   length=4.75
   radius=1.75
   shapeName=dyn_left_hair_b_capsule
   center=<2.4848 0 -0.1221>
   rotation=<rot -87.1875 -179.999 179.9988>
  }]
 dyn_left_hair_bs1 [{shapeType=Capsule, length=2, radius=1, shapeName=dyn_left_hair_bs1_capsule, center=<2.25 0 0>, rotation=<rot 90 0 0>}]
 dyn_left_hair_f [{
   shapeType=Capsule
   length=4.75
   radius=1.75
   shapeName=dyn_left_hair_f_capsule
   center=<2.4848 0 -0.1221>
   rotation=<rot -87.1874 -179.9993 179.9991>
  }]
 dyn_left_hair_fs1 [{shapeType=Capsule, length=2, radius=1, shapeName=dyn_left_hair_fs1_capsule, center=<2.25 0 0>, rotation=<rot 90 0 0>}]
 dyn_left_hair_fs2 [{shapeType=Capsule, length=2, radius=1, shapeName=dyn_left_hair_fs2_capsule, center=<1.5 0 0>, rotation=<rot 90 0 0>}]
 dyn_left_hair_m [{
   shapeType=Capsule
   length=4.75
   radius=1.75
   shapeName=dyn_left_hair_m_capsule
   center=<2.4848 0 -0.1221>
   rotation=<rot -87.1876 -179.9991 179.999>
  }]
 dyn_right_hair_01 [{
   shapeType=Capsule
   length=4.75
   radius=1.75
   shapeName=dyn_right_hair_01_capsule
   center=<2.4848 0 -0.1221>
   rotation=<rot -87.1875 -179.9996 179.9999>
  }]
 dyn_right_hair_s1 [{shapeType=Capsule, length=2, radius=1, shapeName=dyn_right_hair_s1_capsule, center=<2.25 0 0>, rotation=<rot 90 0 0>}]
 dyn_right_hair_s2 [{shapeType=Capsule, length=2, radius=1, shapeName=dyn_right_hair_s2_capsule, center=<2.25 0 0>, rotation=<rot 90 0 0>}]
 head [{shapeType=Capsule, length=8.0176, radius=6.4787, center=<0 -0.11 0>, rotation=<rot 90 0 0>}]
 spine_05 [{shapeType=Capsule, length=18.5, radius=8.375, shapeName=spine_05_capsule, center=<3.3658 0 -1>, rotation=<rot 90 0 0>}]
}
constraints [
 {bone1Name=dyn_right_hair_01, bone2Name=head, swing1LimitDegrees=20, swing2LimitDegrees=15, twistLimitDegrees=2}
 {bone1Name=dyn_right_hair_s1, bone2Name=dyn_right_hair_01, swing1LimitDegrees=45, swing2LimitDegrees=45, twistLimitDegrees=2}
 {bone1Name=dyn_right_hair_s2, bone2Name=dyn_right_hair_01, swing1LimitDegrees=45, swing2LimitDegrees=45, twistLimitDegrees=2}
 {bone1Name=dyn_left_hair_01, bone2Name=head, swing1LimitDegrees=12, swing2LimitDegrees=15, twistLimitDegrees=2}
 {bone1Name=dyn_left_hair_02, bone2Name=dyn_left_hair_01, swing1LimitDegrees=12, swing2LimitDegrees=15, twistLimitDegrees=2}
 {bone1Name=dyn_left_hair_f, bone2Name=dyn_left_hair_02, swing1LimitDegrees=15, swing2LimitDegrees=15, twistLimitDegrees=2}
 {bone1Name=dyn_left_hair_fs1, bone2Name=dyn_left_hair_f, swing1LimitDegrees=15, swing2LimitDegrees=15, twistLimitDegrees=2}
 {bone1Name=dyn_left_hair_fs2, bone2Name=dyn_left_hair_f, swing1LimitDegrees=15, swing2LimitDegrees=15, twistLimitDegrees=2}
 {bone1Name=dyn_left_hair_m, bone2Name=dyn_left_hair_02, swing1LimitDegrees=15, swing2LimitDegrees=15, twistLimitDegrees=2}
 {bone1Name=dyn_left_hair_b, bone2Name=dyn_left_hair_02, swing1LimitDegrees=15, swing2LimitDegrees=15, twistLimitDegrees=2}
 {bone1Name=dyn_left_hair_bs1, bone2Name=dyn_left_hair_b, swing1LimitDegrees=45, swing2LimitDegrees=45, twistLimitDegrees=2}
 {bone1Name=dyn_back_hair_01, bone2Name=head, swing1LimitDegrees=10, swing2LimitDegrees=10, twistLimitDegrees=2}
 {bone1Name=dyn_back_hair_02, bone2Name=dyn_back_hair_01, swing1LimitDegrees=10, swing2LimitDegrees=10, twistLimitDegrees=2}
 {bone1Name=dyn_back_hair_s1, bone2Name=dyn_back_hair_02, swing1LimitDegrees=10, swing2LimitDegrees=45, twistLimitDegrees=2}
 {bone1Name=dyn_back_hair_s2, bone2Name=dyn_back_hair_02, swing1LimitDegrees=10, swing2LimitDegrees=45, twistLimitDegrees=2}
 {bone1Name=dyn_back_hair_s3, bone2Name=dyn_back_hair_02, swing1LimitDegrees=10, swing2LimitDegrees=45, twistLimitDegrees=2}
 {
  bone1Name=head
  bone2Name=spine_05
  swing1LimitDegrees=45
  swing1Motion=Free
  swing2LimitDegrees=45
  swing2Motion=Free
  twistLimitDegrees=45
  twistMotion=Free
 }
]
